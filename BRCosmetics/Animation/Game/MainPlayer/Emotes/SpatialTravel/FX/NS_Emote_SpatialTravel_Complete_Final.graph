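# Niagara effect for the SpatialTravel emote that briefly flashes glowing mesh particles, apparently cube clicks, on a skeletal mesh.
# The single CubeClicks emitter burst-spawns short-lived particles at skeletal mesh locations, with a downward offset lerped by User.CubeGlowStep.
# A curve over emitter loop age fades alpha in and out, while User.AltTimeline switches the color from cyan to red.
# Scaling by User.CubeScale is present but disabled.
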NiagaraSystem /BRCosmetics/Animation/Game/MainPlayer/Emotes/SpatialTravel/FX/NS_Emote_SpatialTravel_Complete_Final
dependencies {
 usedDataInterfaces [&NiagaraDataInterfaceCurve, &NiagaraDataInterfaceArrayColor]
 usedDynamicInputs []
 usedModules [
  &SystemState
  &self:SystemSpawnScript.SystemScriptSource.SystemScriptGraph.NiagaraNodeAssignment_0.SetVariables_#1
 ]
}
emitters {CubeClicks {
  emitterUpdateScript [
   &EmitterState(Life Cycle Mode=::NewEnumerator1, Loop Behavior=::NewEnumerator1, Loop Duration=0.2, UseLoopDelay=false)
   &SpawnBurst_Instantaneous(Use Spawn Probability=false, UseLoopCountLimit=false) {module=SpawnBurst_Instantaneous001}
   &SpawnBurst_Instantaneous(Use Spawn Probability=false, UseLoopCountLimit=false) {enabled=false}
   /SetVariables(Emitter.CubeGlow:NiagaraFloat=&FloatFromCurve(
     FloatCurve=DataInterface() {propertyValues {Curve {
        defaultValue=1
        keys [
         {interpMode=RCIM_Cubic, tangentMode=RCTM_User, time=0, value=0}
         {interpMode=RCIM_Cubic, tangentMode=RCTM_User, time=0.1, value=1}
         {arriveTangent=-1.1111, interpMode=RCIM_Cubic, leaveTangent=-1.1111, time=0.6, value=1}
         {tangentMode=RCTM_User, time=1, value=0}
        ]
       }}}
     CurveIndex=$Emitter.NormalizedLoopAge
    )) {isSetParameters=true}
  ]
  particleSpawnScript [
   &InitializeParticle(
    Color=&SelectLinearColorFromArray(
     Array Sampling Mode=::NewEnumerator1
     Direct Array Index=&MakeIntFromFloat(Float=$User.AltTimeline)
     Color Selection Array=DataInterface() {propertyValues {ColorData [<rgba 0.3497 0.8584 1 1>, <rgba 1 0 0.02 0.5>]}}
     Random Seed=?
    )
    Lifetime=0.2
    Mass Max=1.5
    Mass Mode=::NewEnumerator3
    Mesh Scale Mode=::NewEnumerator3
    Sprite Rotation Mode=::NewEnumerator2
    Sprite Size Mode=::NewEnumerator1
    Uniform Sprite Size Max=8
    Uniform Sprite Size Min=4
    UsePositionOffset=true
    Write Material Random=true
   )
   &SkeletalMeshLocation(
    Bone Sampling Mode=::NewEnumerator4
    MaskSpawnGroup=false
    Sampled Position Offset=&Lerp_Vector(A=<0 0 -5>, B=<1 1 -100>, Alpha=$User.CubeGlowStep)
    Skeletal Mesh=$System.Skeletal Mesh
   )
  ]
  particleUpdateScript [
   &ParticleState()
   &ScaleMeshSize(Scale Factor=&VectorFromFloat(Value=$User.CubeScale)) {enabled=false}
   /SetVariables() {isSetParameters=true}
   &DynamicMaterialParameters(
    Index 0 Param 1=1
    Module.Index0Float0Write=true
    Module.Index0Float1Write=true
    Module.Index0Float2Write=true
    Module.Index0Float3Write=true
   )
   &Color(Color=&SelectLinearColorFromArray(
     Array Sampling Mode=::NewEnumerator1
     Direct Array Index=&MakeIntFromFloat(Float=$User.AltTimeline)
     Color Selection Array=DataInterface() {propertyValues {ColorData [<rgba 0.3497 0.8584 1 1>, <rgba 2 0 0.04 1>]}}
     Random Seed=?
    ))
   &ScaleColor(
    Scale Alpha=$Emitter.CubeGlow
    Scale RGB=&Lerp_Vector(A=<0.3497 0.8584 1>, B=<2 0 0.04>, Alpha=$User.AltTimeline)
    ScaleA=true
    ScaleRGB=true
   )
  ]
  rendererClasses [&NiagaraMeshRendererProperties]
  renderers [
   {rendererClass=&NiagaraMeshRendererProperties, rendererIndex=0}
   {rendererClass=&NiagaraMeshRendererProperties, rendererIndex=1}
  ]
 }}
summary {
 systemName=NS_Emote_SpatialTravel_Complete_Final
 emitters [{emitterName=CubeClicks, rendererClasses [&NiagaraMeshRendererProperties]}]
 userVariables [
  User.AltTimeline {type=NiagaraFloat, defaultValue=0, description=""}
  User.bIsFrontEnd {type=NiagaraBool}
  User.bIsFrontEndPreview {type=NiagaraBool}
  User.CubeFalling {type=NiagaraFloat, defaultValue=0, description=""}
  User.CubeGlow {type=NiagaraFloat, defaultValue=0, description=""}
  User.CubeGlowStep {type=NiagaraFloat, defaultValue=0, description=""}
  User.CubeScale {type=NiagaraFloat, defaultValue=1, description=""}
  User.CubeStep {type=NiagaraFloat, defaultValue=0, description=""}
  User.FrontEnd_color {type=LinearColor, description="", defaultValue {struct=LinearColor, value=<rgba 1 0.6743 0.6445 1>}}
  User.Height_Location {type=NiagaraFloat, defaultValue=17, description=""}
  User.SkeletalMesh {
   type=NiagaraDataInterfaceSkeletalMesh
   defaultValue {
    struct=NiagaraExt_VariableValue_DataInterface
    value {dataInterface=&self:NiagaraDataInterfaceSkeletalMesh_0}
   }
  }
 ]
}
