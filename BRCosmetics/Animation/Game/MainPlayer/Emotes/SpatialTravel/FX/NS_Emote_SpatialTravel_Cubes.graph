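# Niagara system for the SpatialTravel emote's cube effects, using mesh particles placed on skeletal mesh bones or sockets.
# Two CubeClicks emitters draw glows, tinted cyan or red by User.AltTimeline, with CubeGlow and CubeGlowStep driving one's brightness and mesh.
# CubeTrails sheds particles at the CubeStep-selected bone while CubeFalling is raised.
# Two timed StartingVortex emitters add a swirling, attracted swarm, then an expanding sphere burst, sharing the trails' front-end color override.
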NiagaraSystem /BRCosmetics/Animation/Game/MainPlayer/Emotes/SpatialTravel/FX/NS_Emote_SpatialTravel_Cubes
dependencies {
 usedDataInterfaces [
  &NiagaraDataInterfaceCurve
  &NiagaraDataInterfaceArrayFloat
  &NiagaraDataInterfaceArrayColor
  &NiagaraDataInterfaceArrayInt32
  &NiagaraDataInterfaceArrayFloat3
  &NiagaraDataInterfaceSkeletalMesh
 ]
 usedDynamicInputs []
 usedModules [
  &SystemState
  &self:SystemSpawnScript.SystemScriptSource.SystemScriptGraph.NiagaraNodeAssignment_0.SetVariables_#1
 ]
}
emitters {
 CubeClicks {
  emitterUpdateScript [
   &EmitterState(Life Cycle Mode=::NewEnumerator1, UseLoopDelay=false)
   &SpawnBurst_Instantaneous(Use Spawn Probability=false, UseLoopCountLimit=false) {enabled=false}
   &SpawnRate(
    SpawnRate=&FloatFromCurve(FloatCurve=DataInterface() {propertyValues {Curve {keys [<at 0 1>,<at 1 1>,<at 2 0>]}}}, CurveIndex=$Emitter.Age)
    Use Spawn Probability=false
   )
   /SetVariables(Emitter.CubeGlow:NiagaraFloat=&RemapRange(Input Value To Remap=$User.CubeGlow, Input Min=0, Input Max=1, Output Min=0, Output Max=2)) {isSetParameters=true, module=SetVariables_#2}
  ]
  particleSpawnScript [&InitializeParticle(
    Color=&SelectLinearColorFromArray(
     Array Sampling Mode=::NewEnumerator1
     Direct Array Index=&MakeIntFromFloat(Float=$User.AltTimeline)
     Color Selection Array=DataInterface() {propertyValues {ColorData [<rgba 0.3497 0.8584 1 0.5>, <rgba 1 0 0.02 1>]}}
     Random Seed=?
    )
    Mass Max=1.5
    Mass Mode=::NewEnumerator3
    Mesh Scale Mode=::NewEnumerator3
    Sprite Rotation Mode=::NewEnumerator2
    Sprite Size Mode=::NewEnumerator1
    Uniform Sprite Size Max=8
    Uniform Sprite Size Min=4
    UsePositionOffset=false
    Write Material Random=true
   )]
  particleUpdateScript [
   &ParticleState(Kill Particles When Lifetime Has Elapsed=false)
   &SkeletalMeshLocation(
    Bone / Socket Index=&SelectIntFromArray(
     Direct Array Index=&MakeIntFromFloat(Float=$User.AltTimeline)
     Int Selection Array=DataInterface() {propertyValues {IntData [0,6]}}
    )
    Bone Sampling Mode=::NewEnumerator4
    MaskSpawnGroup=false
    Sampled Position Offset=&SelectVectorFromArray(
     Direct Array Index=&MakeIntFromFloat(Float=$User.AltTimeline)
     Vector Selection Array=DataInterface() {propertyValues {FloatData [<0 0 -6>, <0 -12.6 0.5>]}}
    )
    Skeletal Mesh=$System.Skeletal Mesh
   )
   &ScaleMeshSize(Scale Factor=&VectorFromFloat(Value=$User.CubeScale)) {enabled=false}
   /SetVariables(Particles.MeshIndex=&MakeIntFromFloat(Float=$User.CubeGlowStep)) {isSetParameters=true, module=SetVariables_#3}
   &DynamicMaterialParameters(
    Index 0 Param 1=1
    Module.Index0Float0Write=true
    Module.Index0Float1Write=true
    Module.Index0Float2Write=true
    Module.Index0Float3Write=true
   )
   &Color(Color=&SelectLinearColorFromArray(
     Array Sampling Mode=::NewEnumerator1
     Direct Array Index=&MakeIntFromFloat(Float=$User.AltTimeline)
     Color Selection Array=DataInterface() {propertyValues {ColorData [<rgba 0.3497 0.8584 1 1>, <rgba 2 0 0.04 1>]}}
     Random Seed=?
    ))
   &ScaleColor(
    Scale Alpha=$Emitter.CubeGlow
    Scale RGB=&Lerp_Vector(A=<0.3497 0.8584 1>, B=<2 0 0.04>, Alpha=$User.AltTimeline)
    ScaleA=true
    ScaleRGB=true
   )
  ]
  rendererClasses [&NiagaraMeshRendererProperties]
  renderers [
   {rendererClass=&NiagaraMeshRendererProperties, rendererIndex=0}
   {rendererClass=&NiagaraMeshRendererProperties, rendererIndex=1}
  ]
 }
 CubeClicks001 {
  emitterUpdateScript [
   &EmitterState(Life Cycle Mode=::NewEnumerator1, Loop Behavior=::NewEnumerator1, Loop Duration=1.5, UseLoopDelay=false)
   &SpawnBurst_Instantaneous(Spawn Time=0.8, Use Spawn Probability=false, UseLoopCountLimit=false) {module=SpawnBurst_Instantaneous001}
   &SpawnBurst_Instantaneous(Use Spawn Probability=false, UseLoopCountLimit=false) {enabled=false}
   /SetVariables(Emitter.CubeGlow:NiagaraFloat=&FloatFromCurve(
     FloatCurve=DataInterface() {propertyValues {Curve {
        defaultValue=1
        keys [
         {interpMode=RCIM_Cubic, tangentMode=RCTM_User, time=0, value=0}
         {interpMode=RCIM_Cubic, tangentMode=RCTM_User, time=0.1, value=1}
         {arriveTangent=-1.1111, interpMode=RCIM_Cubic, leaveTangent=-1.1111, time=0.6, value=1}
         {tangentMode=RCTM_User, time=1, value=0}
        ]
       }}}
     CurveIndex=$Emitter.NormalizedLoopAge
    )) {isSetParameters=true, module=SetVariables_#2}
  ]
  particleSpawnScript [
   &InitializeParticle(
    Color=&SelectLinearColorFromArray(
     Array Sampling Mode=::NewEnumerator1
     Direct Array Index=&MakeIntFromFloat(Float=$User.AltTimeline)
     Color Selection Array=DataInterface() {propertyValues {ColorData [<rgba 0.3497 0.8584 1 1>, <rgba 1 0 0.02 0.5>]}}
     Random Seed=?
    )
    Lifetime=0.2
    Mass Max=1.5
    Mass Mode=::NewEnumerator3
    Mesh Scale Mode=::NewEnumerator3
    Sprite Rotation Mode=::NewEnumerator2
    Sprite Size Mode=::NewEnumerator1
    Uniform Sprite Size Max=8
    Uniform Sprite Size Min=4
    UsePositionOffset=true
    Write Material Random=true
   )
   &SkeletalMeshLocation(
    Bone / Socket Index=9
    Bone Sampling Mode=::NewEnumerator4
    MaskSpawnGroup=false
    Sampled Position Offset=<0 4 -45.5>
    Skeletal Mesh=$System.Skeletal Mesh
   )
   &InitialMeshOrientation(Mesh Orientation Mode=::NewEnumerator2, Rotation=<0 0 -0.053>, Use Rotation Vector=true)
  ]
  particleUpdateScript [
   &ParticleState()
   &ScaleMeshSize(Scale Factor=&VectorFromFloat(Value=$User.CubeScale)) {enabled=false}
   /SetVariables() {isSetParameters=true, module=SetVariables_#3}
   &DynamicMaterialParameters(
    Index 0 Param 1=1
    Module.Index0Float0Write=true
    Module.Index0Float1Write=true
    Module.Index0Float2Write=true
    Module.Index0Float3Write=true
   )
   /SetVariables(Particles.float:NiagaraFloat=&FloatFromCurve(FloatCurve=DataInterface() {propertyValues {Curve {
        defaultValue=1
        keys [
         {interpMode=RCIM_Cubic, tangentMode=RCTM_User, time=0, value=0}
         {interpMode=RCIM_Cubic, tangentMode=RCTM_User, time=0.1, value=1}
         {arriveTangent=-1.1111, interpMode=RCIM_Cubic, leaveTangent=-1.1111, time=0.6, value=1}
         {tangentMode=RCTM_User, time=1, value=0}
        ]
       }}})) {isSetParameters=true}
   &Color(Color=&SelectLinearColorFromArray(
     Array Sampling Mode=::NewEnumerator1
     Direct Array Index=&MakeIntFromFloat(Float=$User.AltTimeline)
     Color Selection Array=DataInterface() {propertyValues {ColorData [<rgba 0.3497 0.8584 1 1>, <rgba 2 0 0.04 1>]}}
     Random Seed=?
    ))
   &ScaleColor(
    Scale Alpha=$Particles.float
    Scale RGB=&Lerp_Vector(A=<0.3497 0.8584 1>, B=<2 0 0.04>, Alpha=$User.AltTimeline)
    ScaleA=true
    ScaleRGB=true
   )
  ]
  rendererClasses [&NiagaraMeshRendererProperties]
  renderers [
   {rendererClass=&NiagaraMeshRendererProperties, rendererIndex=0}
   {rendererClass=&NiagaraMeshRendererProperties, rendererIndex=1}
  ]
 }
 CubeTrails {
  emitterUpdateScript [
   &EmitterState(Inactive Response=::NewEnumerator1, Life Cycle Mode=::NewEnumerator1, Loop Duration=7, UseLoopDelay=false)
   &SpawnRate(SpawnRate=&Lerp_Float(A=0, B=128, Alpha=$User.CubeFalling), Use Spawn Probability=false)
  ]
  particleSpawnScript [
   /Niagara/Modules/Spawn/Initialization/InitializeParticle(
    Color=&MakeCustomLinearColorFromBool(
     Bool=/Niagara/DynamicInputs/Random/RandomBool(Randomness Mode:ENiagaraRandomnessMode=::NewEnumerator0, Random Seed:NiagaraInt32=0, Fixed Random Seed:NiagaraBool=false)
     True Color=&MakeCustomLinearColorFromBool(
      Bool=&RandomBool(Randomness Mode:ENiagaraRandomnessMode=::NewEnumerator0, Random Seed:NiagaraInt32=0, Fixed Random Seed:NiagaraBool=false)
      True Color=<rgba 0.043 2.0396 10 2>
      False Color=<rgba 0 0.6138 4 2>
     )
     False Color=<rgba 0.2604 3.093 15 2>
    )
    Lifetime=&UniformRangedFloat(Minimum=0.3, Maximum=0.4)
    Mesh Scale=&Multiply_Vector(
     A=<0.7 0.7 0.7>
     B=&VectorFromFloat(Value=&UniformRangedFloat(
       Minimum=&Multiply_Float(
        A=0.15
        B=&FloatFromCurve(
         FloatCurve=DataInterface() {propertyValues {Curve {keys [<at 0 1>,<at 1 1>]}}}
         CurveIndex=&Divide_Float(A=&VectorLength(Vector=$Engine.Owner.Velocity), B=9000)
        )
       )
       Maximum=&Multiply_Float(
        A=0.25
        B=&FloatFromCurve(
         FloatCurve=DataInterface() {propertyValues {Curve {keys [<at 0 1>,<at 1 1.2>]}}}
         CurveIndex=&Divide_Float(A=&VectorLength(Vector=$Engine.Owner.Velocity), B=9000)
        )
       )
      ))
    )
    Module.Write Color=true
    Module.Write Lifetime=true
    Module.Write Mass=true
    Module.Write Material Random=true
    Module.Write Position=false
    Module.Write Scale=true
    Module.Write SpriteRotation=true
    Module.Write SpriteSize=false
    Position=&SimulationPosition()
   ) {enabled=false, module=InitializeParticle002}
   /Niagara/Modules/Spawn/Initialization/V2/InitializeParticle(
    Color=&MakeCustomLinearColorFromBool(
     Bool=&RandomBool(Randomness Mode:ENiagaraRandomnessMode=::NewEnumerator0, Random Seed:NiagaraInt32=0, Fixed Random Seed:NiagaraBool=false)
     True Color=&MakeCustomLinearColorFromBool(
      Bool=&RandomBool(Randomness Mode:ENiagaraRandomnessMode=::NewEnumerator0, Random Seed:NiagaraInt32=0, Fixed Random Seed:NiagaraBool=false)
      True Color=<rgba 0.043 2.0396 10 2>
      False Color=<rgba 0 0.6138 4 2>
     )
     False Color=<rgba 0.2604 3.093 15 2>
    )
    Lifetime=&UniformRangedFloat(Minimum=0.3, Maximum=0.4)
    Mesh Scale Mode=::NewEnumerator3
    Mesh Uniform Scale=&MakeFloatFromVector(
     Vector=&Multiply_Vector(
      A=<0.7 0.7 0.7>
      B=&VectorFromFloat(Value=&UniformRangedFloat(
        Minimum=&Multiply_Float(
         A=0.15
         B=&FloatFromCurve(
          FloatCurve=DataInterface() {propertyValues {Curve {keys [<at 0 1>,<at 1 1>]}}}
          CurveIndex=&Divide_Float(A=&VectorLength(Vector=$Engine.Owner.Velocity), B=9000)
         )
        )
        Maximum=&Multiply_Float(
         A=0.25
         B=&FloatFromCurve(
          FloatCurve=DataInterface() {propertyValues {Curve {keys [<at 0 1>,<at 1 1.2>]}}}
          CurveIndex=&Divide_Float(A=&VectorLength(Vector=$Engine.Owner.Velocity), B=9000)
         )
        )
       ))
     )
     Channel=::NewEnumerator0
    )
    UsePositionOffset=false
    Write Material Random=true
   )
   &SkeletalMeshLocation(
    Bone / Socket Index=&MakeIntFromFloat(Float=$User.CubeStep)
    Bone Sampling Mode=::NewEnumerator4
    MaskSpawnGroup=false
    Sampled Position Offset=&Multiply_VectorByFloat(
     Vector=&RandomRangeVector(Minimum=<1 1 0.15>, Maximum=<-1 -1 -0.15>)
     Float=&SelectFloatFromArray(
      Direct Array Index=&MakeIntFromFloat(Float=$User.CubeStep)
      Float Selection Array=DataInterface() {propertyValues {FloatData [20,5,15,7,7,4,3]}}
     )
    )
    Skeletal Mesh=$System.Skeletal Mesh
   ) {module=SkeletalMeshLocation001}
   &AddVelocity(Velocity=&Multiply_Vector(
     A=&UniformRangedVector(Minimum=<0 0 -150>, Maximum=<1 1 0>)
     B=&VectorFromFloat(Value=&FloatFromCurve(
       FloatCurve=DataInterface() {propertyValues {Curve {keys [<at 0 1>,<at 1 5>]}}}
       CurveIndex=&Divide_Float(A=&VectorLength(Vector=$Engine.Owner.Velocity), B=9000)
      ))
    ))
   &InheritVelocity(Speed Limit=true, UseSpeedThreshold=false)
  ]
  particleUpdateScript [
   &ParticleState()
   &Drag(Drag=2, Use Linear Drag=true, Use Rotational Drag=false)
   &SolveForcesAndVelocity(Clamp Velocity=false, Limit Acceleration=false)
   &ScaleColor(
    Scale RGB=&FN_FrontEndVector3Modifier(
     Gameplay=&MakeVectorFromLinearColorRGB(Color=<rgba 0 0.5347 0.7 1>)
     Front End=&MakeVectorFromLinearColorRGB(Color=$User.FrontEnd_color)
    )
    ScaleA=true
    ScaleRGB=true
   ) {module=ScaleColor002}
   &ScaleMeshSize(Scale Factor=&VectorFromFloat(Value=&FloatFromCurve(FloatCurve=DataInterface() {propertyValues {Curve {keys [{arriveTangent=-1.3985, interpMode=RCIM_Cubic, leaveTangent=-1.3985, tangentMode=RCTM_User, time=0, value=1}, <at 0.9998 0.2>]}}}))) {module=ScaleMeshSize002}
   &CameraOffset()
  ]
  rendererClasses [&NiagaraMeshRendererProperties]
  renderers [{rendererClass=&NiagaraMeshRendererProperties, rendererIndex=0}]
 }
 StartingVortex {
  emitterUpdateScript [
   &EmitterState(
    Inactive Response=::NewEnumerator1
    Life Cycle Mode=::NewEnumerator1
    Loop Behavior=::NewEnumerator1
    Loop Delay=0.2
    Loop Duration=0.6
    UseLoopDelay=true
   )
   &SpawnRate(
    SpawnRate=&FloatFromCurve(FloatCurve=DataInterface() {propertyValues {Curve {keys [<at 0 1>,<at 1 0>]}}}, CurveIndex=$Emitter.NormalizedLoopAge, Scale Curve=200)
    Use Spawn Probability=false
   )
  ]
  particleSpawnScript [
   /Niagara/Modules/Spawn/Initialization/InitializeParticle(
    Color=&MakeCustomLinearColorFromBool(
     Bool=&RandomBool(Randomness Mode:ENiagaraRandomnessMode=::NewEnumerator0, Random Seed:NiagaraInt32=0, Fixed Random Seed:NiagaraBool=false)
     True Color=&MakeCustomLinearColorFromBool(
      Bool=&RandomBool(Randomness Mode:ENiagaraRandomnessMode=::NewEnumerator0, Random Seed:NiagaraInt32=0, Fixed Random Seed:NiagaraBool=false)
      True Color=<rgba 0.043 2.0396 10 2>
      False Color=<rgba 0 0.6138 4 2>
     )
     False Color=<rgba 0.2604 3.093 15 2>
    )
    Lifetime=&UniformRangedFloat(Minimum=0.3, Maximum=0.4)
    Mesh Scale=&Multiply_Vector(
     A=<0.7 0.7 0.7>
     B=&VectorFromFloat(Value=&UniformRangedFloat(
       Minimum=&Multiply_Float(
        A=0.15
        B=&FloatFromCurve(
         FloatCurve=DataInterface() {propertyValues {Curve {keys [<at 0 1>,<at 1 1>]}}}
         CurveIndex=&Divide_Float(A=&VectorLength(Vector=$Engine.Owner.Velocity), B=9000)
        )
       )
       Maximum=&Multiply_Float(
        A=0.25
        B=&FloatFromCurve(
         FloatCurve=DataInterface() {propertyValues {Curve {keys [<at 0 1>,<at 1 1.2>]}}}
         CurveIndex=&Divide_Float(A=&VectorLength(Vector=$Engine.Owner.Velocity), B=9000)
        )
       )
      ))
    )
    Module.Write Color=true
    Module.Write Lifetime=true
    Module.Write Mass=true
    Module.Write Material Random=true
    Module.Write Position=false
    Module.Write Scale=true
    Module.Write SpriteRotation=true
    Module.Write SpriteSize=false
    Position=&SimulationPosition()
   ) {enabled=false, module=InitializeParticle002}
   /Niagara/Modules/Spawn/Initialization/V2/InitializeParticle(
    Color=&MakeCustomLinearColorFromBool(
     Bool=&RandomBool(Randomness Mode:ENiagaraRandomnessMode=::NewEnumerator0, Random Seed:NiagaraInt32=0, Fixed Random Seed:NiagaraBool=false)
     True Color=&MakeCustomLinearColorFromBool(
      Bool=&RandomBool(Randomness Mode:ENiagaraRandomnessMode=::NewEnumerator0, Random Seed:NiagaraInt32=0, Fixed Random Seed:NiagaraBool=false)
      True Color=<rgba 0.043 2.0396 10 2>
      False Color=<rgba 0 0.6138 4 2>
     )
     False Color=<rgba 0.2604 3.093 15 2>
    )
    Lifetime=&UniformRangedFloat(Minimum=0.3, Maximum=0.4)
    Mesh Scale Mode=::NewEnumerator3
    Mesh Uniform Scale=&MakeFloatFromVector(
     Vector=&Multiply_Vector(
      A=<0.7 0.7 0.7>
      B=&VectorFromFloat(Value=&UniformRangedFloat(
        Minimum=&Multiply_Float(
         A=0.15
         B=&FloatFromCurve(
          FloatCurve=DataInterface() {propertyValues {Curve {keys [<at 0 1>,<at 1 1>]}}}
          CurveIndex=&Divide_Float(A=&VectorLength(Vector=$Engine.Owner.Velocity), B=9000)
         )
        )
        Maximum=&Multiply_Float(
         A=0.25
         B=&FloatFromCurve(
          FloatCurve=DataInterface() {propertyValues {Curve {keys [<at 0 1>,<at 1 1.2>]}}}
          CurveIndex=&Divide_Float(A=&VectorLength(Vector=$Engine.Owner.Velocity), B=9000)
         )
        )
       ))
     )
     Channel=::NewEnumerator0
    )
    UsePositionOffset=false
    Write Material Random=true
   )
   &SkeletalMeshLocation(
    Bone / Socket Index=&MakeIntFromFloat(Float=9)
    Bone Sampling Mode=::NewEnumerator4
    MaskSpawnGroup=false
    Position Sampling=::NewEnumerator2
    Sampled Position Offset=<0 0 -20>
    Skeletal Mesh=$System.Skeletal Mesh
   ) {module=SkeletalMeshLocation001}
   &ShapeLocation(
    MaskSpawnGroup=false
    Override Seed=false
    Shape Origin=$Particles.SkeletalMeshLocation001.SampledPosition
    Sphere Radius=50
    UseOwnerScale=true
   )
   &AddVelocity(Velocity=&Multiply_Vector(
     A=&UniformRangedVector(Minimum=<0 0 0>, Maximum=<1 1 0>)
     B=&VectorFromFloat(Value=&FloatFromCurve(
       FloatCurve=DataInterface() {propertyValues {Curve {keys [<at 0 1>,<at 1 5>]}}}
       CurveIndex=&Divide_Float(A=&VectorLength(Vector=$Engine.Owner.Velocity), B=9000)
      ))
    ))
  ]
  particleUpdateScript [
   &ParticleState()
   &Drag(Drag=2, Use Linear Drag=true, Use Rotational Drag=false)
   &VortexForce(
    UseOriginOffset=false
    Vortex Force Amount=&RandomRangeFloat(Minimum=800, Maximum=3000)
    Vortex Origin:NiagaraPosition=$Particles.SkeletalMeshLocation001.SampledPosition
   )
   &PointAttractionForce(
    Attraction Strength=&FloatFromCurve(
     FloatCurve=DataInterface() {propertyValues {Curve {defaultValue=1, keys [{tangentMode=RCTM_User, time=0, value=0}, {tangentMode=RCTM_User, time=1, value=1}]}}}
     Scale Curve=1000
    )
    Kill Within Radius=false
    Module.Use Falloff=true
    UseAttractorOffset=false
   )
   &SolveForcesAndVelocity(Clamp Velocity=false, Limit Acceleration=false)
   &ScaleColor(
    Scale RGB=&FN_FrontEndVector3Modifier(
     Gameplay=&MakeVectorFromLinearColorRGB(Color=<rgba 0 0.5347 0.7 1>)
     Front End=&MakeVectorFromLinearColorRGB(Color=$User.FrontEnd_color)
    )
    ScaleA=true
    ScaleRGB=true
   ) {module=ScaleColor002}
   &ScaleMeshSize(Scale Factor=&VectorFromFloat(Value=&FloatFromCurve(FloatCurve=DataInterface() {propertyValues {Curve {keys [{arriveTangent=-1.3985, interpMode=RCIM_Cubic, leaveTangent=-1.3985, tangentMode=RCTM_User, time=0, value=1}, <at 0.9998 0.2>]}}}))) {module=ScaleMeshSize002}
   &CameraOffset()
  ]
  rendererClasses [&NiagaraMeshRendererProperties]
  renderers [{rendererClass=&NiagaraMeshRendererProperties, rendererIndex=0}]
 }
 StartingVortex001 {
  emitterUpdateScript [
   &EmitterState(
    Inactive Response=::NewEnumerator1
    Life Cycle Mode=::NewEnumerator1
    Loop Behavior=::NewEnumerator1
    Loop Delay=0.8
    Loop Duration=0.4
    UseLoopDelay=true
   )
   &SpawnRate(
    SpawnRate=&FloatFromCurve(FloatCurve=DataInterface() {propertyValues {Curve {keys [<at 0 1>,<at 0.5 0>]}}}, CurveIndex=$Emitter.NormalizedLoopAge, Scale Curve=500)
    Use Spawn Probability=false
   )
  ]
  particleSpawnScript [
   /Niagara/Modules/Spawn/Initialization/InitializeParticle(
    Color=&MakeCustomLinearColorFromBool(
     Bool=&RandomBool(Randomness Mode:ENiagaraRandomnessMode=::NewEnumerator0, Random Seed:NiagaraInt32=0, Fixed Random Seed:NiagaraBool=false)
     True Color=&MakeCustomLinearColorFromBool(
      Bool=&RandomBool(Randomness Mode:ENiagaraRandomnessMode=::NewEnumerator0, Random Seed:NiagaraInt32=0, Fixed Random Seed:NiagaraBool=false)
      True Color=<rgba 0.043 2.0396 10 2>
      False Color=<rgba 0 0.6138 4 2>
     )
     False Color=<rgba 0.2604 3.093 15 2>
    )
    Lifetime=&UniformRangedFloat(Minimum=0.2, Maximum=0.35)
    Mesh Scale=&Multiply_Vector(
     A=<0.7 0.7 0.7>
     B=&VectorFromFloat(Value=&UniformRangedFloat(
       Minimum=&Multiply_Float(
        A=0.15
        B=&FloatFromCurve(
         FloatCurve=DataInterface() {propertyValues {Curve {keys [<at 0 1>,<at 1 1>]}}}
         CurveIndex=&Divide_Float(A=&VectorLength(Vector=$Engine.Owner.Velocity), B=9000)
        )
       )
       Maximum=&Multiply_Float(
        A=0.25
        B=&FloatFromCurve(
         FloatCurve=DataInterface() {propertyValues {Curve {keys [<at 0 1>,<at 1 1.2>]}}}
         CurveIndex=&Divide_Float(A=&VectorLength(Vector=$Engine.Owner.Velocity), B=9000)
        )
       )
      ))
    )
    Module.Write Color=true
    Module.Write Lifetime=true
    Module.Write Mass=true
    Module.Write Material Random=true
    Module.Write Position=false
    Module.Write Scale=true
    Module.Write SpriteRotation=true
    Module.Write SpriteSize=false
    Position=&SimulationPosition()
   ) {module=InitializeParticle002}
   &SkeletalMeshLocation(
    Bone / Socket Index=&MakeIntFromFloat(Float=9)
    Bone Sampling Mode=::NewEnumerator4
    MaskSpawnGroup=false
    Position Sampling=::NewEnumerator2
    Sampled Position Offset=<0 4 0>
    Skeletal Mesh=$System.Skeletal Mesh
   ) {module=SkeletalMeshLocation001}
   &ShapeLocation(
    MaskSpawnGroup=false
    Offset=<0 0 -25>
    Offset Mode=::NewEnumerator0
    Override Seed=false
    Shape Origin=$Particles.SkeletalMeshLocation001.SampledPosition
    Sphere Radius=&FloatFromCurve(
     FloatCurve=DataInterface() {propertyValues {Curve {defaultValue=1, keys [{tangentMode=RCTM_User, time=0, value=0.4}, {tangentMode=RCTM_User, time=0.5, value=1}]}}}
     CurveIndex=$Emitter.NormalizedLoopAge
     Scale Curve=65
    )
    Sphere Surface Distribution=1
    UseOwnerScale=true
   )
   &AddVelocity(Velocity=&Multiply_Vector(
     A=&UniformRangedVector(Minimum=<0 0 0>, Maximum=<1 1 0>)
     B=&VectorFromFloat(Value=&FloatFromCurve(
       FloatCurve=DataInterface() {propertyValues {Curve {keys [<at 0 1>,<at 1 5>]}}}
       CurveIndex=&Divide_Float(A=&VectorLength(Vector=$Engine.Owner.Velocity), B=9000)
      ))
    ))
  ]
  particleUpdateScript [
   &ParticleState()
   &Drag(Drag=2, Use Linear Drag=true, Use Rotational Drag=false)
   &SolveForcesAndVelocity(Clamp Velocity=false, Limit Acceleration=false)
   &ScaleColor(
    Scale Alpha=&FloatFromCurve(FloatCurve=DataInterface() {propertyValues {Curve {keys [<at 0 0>,<at 0.25 1>,<at 0.5 1>,<at 1 0>]}}})
    ScaleA=true
    ScaleRGB=true
   ) {enabled=false}
   &ScaleColor(
    Scale Alpha=&FloatFromCurve(FloatCurve=DataInterface() {propertyValues {Curve {keys [<at 0 0>,<at 0.1 1>,<at 0.9 1>,<at 1 0>]}}})
    Scale RGB=&FN_FrontEndVector3Modifier(Gameplay=<1 1 1>, Front End=&MakeVectorFromLinearColorRGB(Color=<rgba 2.5 0.6315 0.4322 1>))
    ScaleA=true
    ScaleRGB=true
   ) {enabled=false, module=ScaleColor001}
   &ScaleColor(
    Scale RGB=&FN_FrontEndVector3Modifier(
     Gameplay=&MakeVectorFromLinearColorRGB(Color=<rgba 0 0.5347 0.7 1>)
     Front End=&MakeVectorFromLinearColorRGB(Color=$User.FrontEnd_color)
    )
    ScaleA=true
    ScaleRGB=true
   ) {module=ScaleColor002}
   &ScaleMeshSize(Scale Factor=&VectorFromFloat(Value=&FloatFromCurve(FloatCurve=DataInterface() {propertyValues {Curve {keys [
          <at 0 0 RCIM_Cubic>
          {arriveTangent=-1.3985, interpMode=RCIM_Cubic, leaveTangent=-1.3985, tangentMode=RCTM_User, time=0.2112, value=1}
          <at 0.9998 0.2>
         ]}}}))) {module=ScaleMeshSize002}
   &CameraOffset()
  ]
  rendererClasses [&NiagaraMeshRendererProperties]
  renderers [{rendererClass=&NiagaraMeshRendererProperties, rendererIndex=0}]
 }
}
summary {
 systemName=NS_Emote_SpatialTravel_Cubes
 emitters [
  {emitterName=CubeTrails, rendererClasses [&NiagaraMeshRendererProperties]}
  {emitterName=StartingVortex, rendererClasses [&NiagaraMeshRendererProperties]}
  {emitterName=CubeClicks, rendererClasses [&NiagaraMeshRendererProperties]}
  {emitterName=CubeClicks001, rendererClasses [&NiagaraMeshRendererProperties]}
  {emitterName=StartingVortex001, rendererClasses [&NiagaraMeshRendererProperties]}
 ]
 userVariables [
  User.AltTimeline {type=NiagaraFloat, defaultValue=0, description=""}
  User.bIsFrontEnd {type=NiagaraBool}
  User.bIsFrontEndPreview {type=NiagaraBool}
  User.CubeFalling {type=NiagaraFloat, defaultValue=0, description=""}
  User.CubeGlow {type=NiagaraFloat, defaultValue=0, description=""}
  User.CubeGlowStep {type=NiagaraFloat, defaultValue=0, description=""}
  User.CubeScale {type=NiagaraFloat, defaultValue=1, description=""}
  User.CubeStep {type=NiagaraFloat, defaultValue=0, description=""}
  User.FrontEnd_color {type=LinearColor, description="", defaultValue {struct=LinearColor, value=<rgba 1 0.6743 0.6445 1>}}
  User.Height_Location {type=NiagaraFloat, defaultValue=17, description=""}
 ]
}
