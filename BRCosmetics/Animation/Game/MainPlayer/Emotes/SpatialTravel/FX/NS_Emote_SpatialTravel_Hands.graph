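# Niagara system for the hand effects of the Spatial Travel emote, pinned to skeletal mesh bones or sockets.
# Partitioned particles split between two hand points, forming layered glow sprites with lights, light shafts and drifting mesh squares, plus a separate light at a third point.
# User.HandGlow drives intensity, while User.AltTimeline fades out shafts and shifts the glow to one hand.
# The squares switch tint between gameplay and front end.
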NiagaraSystem /BRCosmetics/Animation/Game/MainPlayer/Emotes/SpatialTravel/FX/NS_Emote_SpatialTravel_Hands
dependencies {
 usedDataInterfaces [
  &NiagaraDataInterfaceCurve
  &NiagaraDataInterfaceArrayFloat
  &NiagaraDataInterfaceArrayFloat3
  &NiagaraDataInterfaceArrayInt32
  &NiagaraDataInterfaceSkeletalMesh
 ]
 usedDynamicInputs []
 usedModules [
  &SystemState
  &self:SystemSpawnScript.SystemScriptSource.SystemScriptGraph.NiagaraNodeAssignment_0.SetVariables_#1
 ]
}
emitters {
 BlowingParticles {
  emitterUpdateScript [
   &EmitterState(Life Cycle Mode=::NewEnumerator1, UseLoopDelay=false)
   &SpawnBurst_Instantaneous(Use Spawn Probability=false, UseLoopCountLimit=false)
  ]
  particleSpawnScript [
   &InitializeParticle(
    Color=<rgba 1.276 3.8972 5 1>
    Mass Max=1.5
    Mass Mode=::NewEnumerator3
    Sprite Rotation Mode=::NewEnumerator2
    Sprite Size Mode=::NewEnumerator1
    Uniform Sprite Size Max=8
    Uniform Sprite Size Min=4
    UsePositionOffset=false
    Write Material Random=true
   )
   &SkeletalMeshLocation(
    Bone / Socket Index=2
    Bone Sampling Mode=::NewEnumerator4
    MaskSpawnGroup=false
    Sampled Position Offset=<0 35 80>
    Skeletal Mesh=$System.Skeletal Mesh
   )
  ]
  particleUpdateScript [&ParticleState(Kill Particles When Lifetime Has Elapsed=false)]
  rendererClasses [&NiagaraLightRendererProperties]
  renderers [{rendererClass=&NiagaraLightRendererProperties, rendererIndex=0}]
 }
 Light_Shafts {
  emitterUpdateScript [
   &EmitterState(Life Cycle Mode=::NewEnumerator1, Loop Duration=1.6, UseLoopDelay=false)
   &SpawnRate(
    SpawnRate=&Lerp_Float(A=13, B=0, Alpha=&ClampFloat(Float=$User.AltTimeline))
    Use Spawn Probability=false
   )
  ]
  particleSpawnScript [
   &PartitionParticles()
   /Niagara/Modules/Spawn/Initialization/InitializeParticle(
    Color=&UniformRangedLinearColor(Minimum=<rgba 0.1912 0.5906 0.6038 0.25>, Maximum=<rgba 0.6333 1.9563 2 0.25>)
    Lifetime=0.5
    Module.Write Color=true
    Module.Write Lifetime=true
    Module.Write Mass=true
    Module.Write Material Random=true
    Module.Write Position=false
    Module.Write Scale=false
    Module.Write SpriteRotation=true
    Module.Write SpriteSize=true
    Position=&SimulationPosition()
    Sprite Size=<150 325>
   ) {enabled=false}
   /Niagara/Modules/Spawn/Initialization/V2/InitializeParticle(
    Color=&UniformRangedLinearColor(Minimum=<rgba 0.1912 0.5906 0.6038 0.25>, Maximum=<rgba 0.6333 1.9563 2 0.25>)
    Lifetime=0.5
    Position Offset=&MakeVector(
     X=&FloatFromCurve(
      FloatCurve=DataInterface() {propertyValues {Curve {defaultValue=1, keys [{tangentMode=RCTM_User, time=0, value=0}, {tangentMode=RCTM_User, time=1, value=1}]}}}
      CurveIndex=&Add_Float(
       A=$Particles.NormalizedAge
       B=&SelectFloatFromArray(
        Array Sampling Mode=::NewEnumerator0
        Direct Array Index=?
        Float Selection Array=DataInterface() {propertyValues {FloatData [0,1]}}
        Random Seed=0
       )
      )
     )
     Y=0
     Z=0
    )
    Sprite Size=<50 100>
    Sprite Size Mode=::NewEnumerator4
    UsePositionOffset=true
    Write Material Random=true
   ) {module=InitializeParticle001}
   /Niagara/Modules/Update/SubUV/SubUVAnimation(SubUV Animation Mode=::NewEnumerator1)
   &DynamicMaterialParameters(
    Index 0 Param 1=&UniformRangedFloat(Minimum=0, Maximum=1)
    Module.Param0WriteEnabled=true
    Module.Param1WriteEnabled=true
    Module.Param2WriteEnabled=true
    Module.Param3WriteEnabled=true
   )
   &ApplyInitialForces()
   &InitialMeshOrientation(
    Mesh Orientation Mode=::NewEnumerator2
    Rotation=&Lerp_Vector(
     A=&SelectVectorFromArray(
      Direct Array Index=$Particles.PartitionParticles.Partition
      Vector Selection Array=DataInterface() {propertyValues {FloatData [<-0.25 0.12 0>, <-0.25 -0.12 0>]}}
     )
     B=&SelectVectorFromArray(
      Direct Array Index=$Particles.PartitionParticles.Partition
      Vector Selection Array=DataInterface() {propertyValues {FloatData [<-0.25 0.12 0>, <0 0.15 0>]}}
     )
     Alpha=$User.AltTimeline
    )
    Use Rotation Vector=true
   )
   &AlignSpriteToMeshOrientation(SetAlignment=true, SetFacing=true) {module=AlignSpriteToMeshOrientation001}
  ]
  particleUpdateScript [
   &SolveForcesAndVelocity(Clamp Velocity=false, Limit Acceleration=false)
   &ParticleState()
   &ScaleSpriteSize(
    Scale Factor=&MakeVector2D(
     X=1
     Y=&ClampFloat(Float=&Lerp_Float(A=0, B=&Lerp_Float(A=1, B=0, Alpha=$User.AltTimeline), Alpha=$User.HandGlow))
    )
    Scale Sprite Size Mode=::NewEnumerator2
   ) {module=ScaleSpriteSize001}
   &SkeletalMeshLocation(
    Bone / Socket Index=&SelectIntFromArray(Direct Array Index=$Particles.PartitionParticles.Partition, Int Selection Array=DataInterface() {propertyValues {IntData [0,1]}})
    Bone Sampling Mode=::NewEnumerator4
    MaskSpawnGroup=false
    Skeletal Mesh=$System.Skeletal Mesh
   )
   &SkeletalMeshLocation(
    Bone / Socket Index=2
    Bone Sampling Mode=::NewEnumerator4
    MaskSpawnGroup=false
    Position Sampling=::NewEnumerator2
    Skeletal Mesh=$System.Skeletal Mesh
   ) {module=SkeletalMeshLocation001}
   &AlignSpriteToMeshOrientation(Mesh Orientation Relative Sprite Alignment Vector=&Multiply_Vector(
     A=&Subtract_Vector(A=$Output.SkeletalMeshLocation.SampledPosition, B=/Niagara/DynamicInputs/Helpers/SimulationPosition())
     B=&MakeVector(X=1, Y=1, Z=&Lerp_Float(A=0.35, B=1.35, Alpha=$Particles.MaterialRandom))
    )) {enabled=false}
   &ScaleColor(
    Scale Alpha=&FloatFromCurve(
     FloatCurve=DataInterface() {propertyValues {Curve {keys [
         {arriveTangent=3.2867, interpMode=RCIM_Cubic, leaveTangent=3.2867, tangentMode=RCTM_User, time=0, value=0}
         <at 0.5 1 RCIM_Cubic>
         {arriveTangent=-2.2446, interpMode=RCIM_Cubic, leaveTangent=-2.2446, tangentMode=RCTM_User, time=1, value=0}
        ]}}}
     Scale Curve=2
    )
    ScaleA=true
    ScaleRGB=true
   )
   &ScaleSpriteSize(Scale Factor=&MakeVector2D(
     X=1
     Y=&FloatFromCurve(FloatCurve=DataInterface() {propertyValues {Curve {keys [
          <at 0 0 RCIM_Cubic>
          {interpMode=RCIM_Cubic, tangentMode=RCTM_User, time=0.15, value=1}
          {interpMode=RCIM_Cubic, tangentMode=RCTM_User, time=0.85, value=1}
          <at 1.0023 0 RCIM_Cubic>
         ]}}})
    ))
  ]
  rendererClasses [&NiagaraSpriteRendererProperties, &NiagaraMeshRendererProperties]
  renderers [
   {rendererClass=&NiagaraSpriteRendererProperties}
   {rendererClass=&NiagaraMeshRendererProperties, rendererIndex=1}
  ]
 }
 NE_Glow {
  emitterUpdateScript [
   &EmitterState(
    Inactive Response=::NewEnumerator1
    Life Cycle Mode=::NewEnumerator1
    Loop Behavior=::NewEnumerator1
    Loop Duration Mode=::NewEnumerator1
    UseLoopDelay=false
   )
   &SpawnBurst_Instantaneous(Spawn Count=2, Use Spawn Probability=false, UseLoopCountLimit=false)
  ]
  particleSpawnScript [
   &PartitionParticles()
   &InitializeParticle(
    Lifetime=0.25
    Mass Mode=::NewEnumerator1
    Mesh Scale Mode=::NewEnumerator3
    Mesh Uniform Scale=0.2
    Position Mode=::NewEnumerator1
    Sprite Rotation Mode=::NewEnumerator2
    Sprite Size Mode=::NewEnumerator3
    Uniform Sprite Size=50
    UsePositionOffset=true
    Write Material Random=true
   ) {module=InitializeParticle001}
   &CameraOffset()
  ]
  particleUpdateScript [
   &ParticleState(Kill Particles When Lifetime Has Elapsed=false)
   &SolveForcesAndVelocity(Clamp Velocity=false, Limit Acceleration=false)
   &ScaleColor(
    Scale Alpha=&FloatFromCurve(
     FloatCurve=DataInterface() {propertyValues {Curve {keys [
         {arriveTangent=6.4019, interpMode=RCIM_Cubic, leaveTangent=6.4019, tangentMode=RCTM_User, time=0, value=0}
         {interpMode=RCIM_Cubic, leaveTangent=0.002, time=0.2, value=0.7}
        ]}}}
     Scale Curve=&Lerp_Float(
      A=&RemapRange(Input Value To Remap=&ClampFloat(Float=$User.HandGlow), Input Min=0, Input Max=1, Output Min=0, Output Max=0.1)
      B=&SelectFloatFromArray(Direct Array Index=$Particles.PartitionParticles.Partition, Float Selection Array=DataInterface() {propertyValues {FloatData [0,0.1]}})
      Alpha=&ClampFloat(Float=$User.AltTimeline)
     )
    )
    Scale RGB=&MakeVectorFromLinearColorRGB(Color=<rgba 0.2128 0.6643 1 1>)
    ScaleA=true
    ScaleRGB=true
   )
   &ScaleSpriteSize(Scale Factor=&Multiply_Vector2D_ByFloat(
     Float=&Multiply_Float(
      A=&FloatFromCurve(
       FloatCurve=DataInterface() {propertyValues {Curve {keys [
           {arriveTangent=0.7585, interpMode=RCIM_Cubic, leaveTangent=0.7585, tangentMode=RCTM_User, time=0, value=0.4}
           {interpMode=RCIM_Cubic, tangentMode=RCTM_User, time=1, value=1}
          ]}}}
       Scale Curve=0.7
      )
      B=1
     )
     Vector2D=<1 1>
    ))
   &SpriteRotationRate(Rotation Rate=&UniformRangedFloat(Minimum=-50, Maximum=50))
   &SpriteFacingAndAlignment(
    Sprite Alignment=&TransformVector(Vector=<1 0 0>, Source Space=::World, Destination Space=::Simulation)
    WriteAlignment=true
    WriteFacing=false
   )
   /SetVariables(Particles.DarkGlowSize:Vector2f=&Multiply_Vector2D_ByFloat(Float=1.2, Vector2D=$Particles.SpriteSize)) {isSetParameters=true}
   /SetVariables(Particles.DargGlowColor:LinearColor=&MakeLinearColorFromVectorAndFloat(
     "Vector (RGB)"=&MakeVectorFromLinearColorRGB(Color=<rgba 0.2448 0.5652 1 1>)
     "Float (Alpha)"=&FloatFromCurve(
      FloatCurve=DataInterface() {propertyValues {Curve {keys [
          {arriveTangent=6.4019, interpMode=RCIM_Cubic, leaveTangent=6.4019, tangentMode=RCTM_User, time=0, value=0}
          {interpMode=RCIM_Cubic, leaveTangent=0.002, time=0.2, value=0.7}
         ]}}}
      Scale Curve=&Lerp_Float(
       A=&RemapRange(Input Value To Remap=&ClampFloat(Float=$User.HandGlow), Input Min=0, Input Max=1, Output Min=0, Output Max=0.1)
       B=&SelectFloatFromArray(Direct Array Index=$Particles.PartitionParticles.Partition, Float Selection Array=DataInterface() {propertyValues {FloatData [0,0.1]}})
       Alpha=&ClampFloat(Float=$User.AltTimeline)
      )
     )
    )) {isSetParameters=true}
   /SetVariables(Particles.CentralFlareSize:Vector2f=&Multiply_Vector2D_ByFloat(Float=0.7, Vector2D=$Particles.SpriteSize)) {isSetParameters=true}
   /SetVariables(Particles.CentralFlareColor:LinearColor=&MakeLinearColorFromVectorAndFloat(
     "Vector (RGB)"=&MakeVectorFromLinearColorRGB(Color=<rgba 0.1978 0.9427 1 1>)
     "Float (Alpha)"=&FloatFromCurve(
      FloatCurve=DataInterface() {propertyValues {Curve {keys [
          {arriveTangent=6.4019, interpMode=RCIM_Cubic, leaveTangent=6.4019, tangentMode=RCTM_User, time=0, value=0}
          {interpMode=RCIM_Cubic, leaveTangent=0.002, time=0.2, value=0.7}
         ]}}}
      Scale Curve=&Lerp_Float(
       A=&RemapRange(Input Value To Remap=&ClampFloat(Float=$User.HandGlow), Input Min=0, Input Max=1, Output Min=0, Output Max=1)
       B=&SelectFloatFromArray(Direct Array Index=$Particles.PartitionParticles.Partition, Float Selection Array=DataInterface() {propertyValues {FloatData [0,1]}})
       Alpha=&ClampFloat(Float=$User.AltTimeline)
      )
     )
    )) {isSetParameters=true}
   &FN_Light_Attributes(
    Color=$Particles.CentralFlareColor
    Create Color Attribute=true
    Create Exponent Attribute=true
    Create Radius Attribute=true
    Exponent=3
    Radius=40
    Scale Color=&Lerp_Float(
     A=&RemapRange(Input Value To Remap=&ClampFloat(Float=$User.HandGlow), Input Min=0, Input Max=1, Output Min=0, Output Max=30)
     B=&SelectFloatFromArray(Direct Array Index=$Particles.PartitionParticles.Partition, Float Selection Array=DataInterface() {propertyValues {FloatData [0,1]}})
     Alpha=&ClampFloat(Float=$User.AltTimeline)
    )
   )
   &DynamicMaterialParameters(
    Index 1 Param 1=0.3
    Index 1 Param 3=1
    Module.Index1Float0Write=true
    Module.Index1Float1Write=true
    Module.Index1Float2Write=true
    Module.Index1Float3Write=true
    Module.Index3Float0Write=true
    Module.Index3Float1Write=true
    Module.Index3Float2Write=true
    Module.Index3Float3Write=true
    Write Parameter Index 0=::NewEnumerator0
    Write Parameter Index 1=::NewEnumerator1
    Write Parameter Index 3=::NewEnumerator1
   )
   &FN_Sprite_Attributes(
    Alignment:Vector3f=<1 0 0>
    Create Alignment Attribute=true
    Create Sprite Rotation Attribute=true
    Create SpriteSize Attribute=true
    SpriteSize=&Multiply_Vector2D(A=$Particles.CentralFlareSize, B=<0.25 2>)
    Write to Module Namespace=true
   )
   &SkeletalMeshLocation(
    Bone / Socket Index=$Particles.PartitionParticles.Partition
    Bone Sampling Mode=::NewEnumerator4
    MaskSpawnGroup=false
    Sampled Position Offset=&SelectVectorFromArray(
     Direct Array Index=$Particles.PartitionParticles.Partition
     Vector Selection Array=DataInterface() {propertyValues {FloatData [<0 0 0>, <0 0 0>]}}
    )
    Sampled Position Offset Coordinate Space:ENiagaraSkelMeshTransforms=::NewEnumerator2
    Skeletal Mesh=$System.Skeletal Mesh
   )
  ]
  rendererClasses [&NiagaraSpriteRendererProperties, &NiagaraLightRendererProperties]
  renderers [
   {rendererClass=&NiagaraSpriteRendererProperties}
   {rendererClass=&NiagaraSpriteRendererProperties, rendererIndex=1}
   {rendererClass=&NiagaraLightRendererProperties, rendererIndex=2}
   {rendererClass=&NiagaraSpriteRendererProperties, rendererIndex=3}
  ]
 }
 Square_far001 {
  emitterUpdateScript [
   &EmitterState(
    Inactive Response=::NewEnumerator1
    Life Cycle Mode=::NewEnumerator1
    Loop Behavior=::NewEnumerator1
    Loop Duration Mode=::NewEnumerator1
    UseLoopDelay=false
   )
   &SpawnRate(
    SpawnRate=&FloatFromCurve(
     FloatCurve=DataInterface() {propertyValues {Curve {keys [<at 0 1>,<at 1 0>]}}}
     CurveIndex=$Emitter.NormalizedLoopAge
     Scale Curve=&Lerp_Float(A=0, B=100, Alpha=&ClampFloat(Float=$User.HandGlow))
    )
    Use Spawn Probability=false
   )
  ]
  particleSpawnScript [
   &PartitionParticles()
   /Niagara/Modules/Spawn/Initialization/InitializeParticle(
    Color=&MakeCustomLinearColorFromBool(
     Bool=/Niagara/DynamicInputs/Random/RandomBool(Randomness Mode:ENiagaraRandomnessMode=::NewEnumerator0, Random Seed:NiagaraInt32=0, Fixed Random Seed:NiagaraBool=false)
     True Color=&MakeCustomLinearColorFromBool(
      Bool=&RandomBool(Randomness Mode:ENiagaraRandomnessMode=::NewEnumerator0, Random Seed:NiagaraInt32=0, Fixed Random Seed:NiagaraBool=false)
      True Color=<rgba 0.043 2.0396 10 2>
      False Color=<rgba 0 0.6138 4 2>
     )
     False Color=<rgba 0.2604 3.093 15 2>
    )
    Lifetime=&UniformRangedFloat(Minimum=0.3, Maximum=0.4)
    Mesh Scale=&Multiply_Vector(
     A=<0.7 0.7 0.7>
     B=&VectorFromFloat(Value=&UniformRangedFloat(
       Minimum=&Multiply_Float(
        A=0.1
        B=&FloatFromCurve(
         FloatCurve=DataInterface() {propertyValues {Curve {keys [<at 0 1>,<at 1 1>]}}}
         CurveIndex=&Divide_Float(A=&VectorLength(Vector=$Engine.Owner.Velocity), B=9000)
        )
       )
       Maximum=&Multiply_Float(
        A=0.2
        B=&FloatFromCurve(
         FloatCurve=DataInterface() {propertyValues {Curve {keys [<at 0 1>,<at 1 1.2>]}}}
         CurveIndex=&Divide_Float(A=&VectorLength(Vector=$Engine.Owner.Velocity), B=9000)
        )
       )
      ))
    )
    Module.Write Color=true
    Module.Write Lifetime=true
    Module.Write Mass=true
    Module.Write Material Random=true
    Module.Write Position=false
    Module.Write Scale=true
    Module.Write SpriteRotation=true
    Module.Write SpriteSize=false
    Position=/Niagara/Functions/Localspace/SimulationPosition()
   ) {module=InitializeParticle002}
   &SkeletalMeshLocation(
    Bone / Socket Index=&SelectIntFromArray(Direct Array Index=$Particles.PartitionParticles.Partition, Int Selection Array=DataInterface() {propertyValues {IntData [0,1]}})
    Bone Sampling Mode=::NewEnumerator4
    MaskSpawnGroup=false
    Position Sampling=::NewEnumerator2
    Sampled Position Offset=&SelectVectorFromArray(
     Direct Array Index=$Particles.PartitionParticles.Partition
     Vector Selection Array=DataInterface() {propertyValues {FloatData [<-5 0 0>, <5 1 1>]}}
    )
    Skeletal Mesh=$System.Skeletal Mesh
   ) {module=SkeletalMeshLocation001}
   &ShapeLocation(
    MaskSpawnGroup=false
    Override Seed=false
    Shape Origin=$Particles.SkeletalMeshLocation001.SampledPosition
    Sphere Radius=8
    UseOwnerScale=true
   )
   &AddVelocity(Velocity=&Multiply_Vector(
     A=&UniformRangedVector(Minimum=<0 0 0>, Maximum=<1 1 0>)
     B=&VectorFromFloat(Value=&FloatFromCurve(
       FloatCurve=DataInterface() {propertyValues {Curve {keys [<at 0 1>,<at 1 5>]}}}
       CurveIndex=&Divide_Float(A=&VectorLength(Vector=$Engine.Owner.Velocity), B=9000)
      ))
    ))
  ]
  particleUpdateScript [
   &ParticleState()
   &ScaleMeshSize(Scale Factor=&VectorFromFloat(Value=&MaskFloat(
      Mask Source=1
      Mask Parameter=$Particles.PartitionParticles.Partition
      Mask=0
      Masked Float=&Lerp_Float(A=1, B=0, Alpha=&ClampFloat(Float=$User.AltTimeline))
      Passthrough Float=1
     )))
   &Drag(Drag=2, Use Linear Drag=true, Use Rotational Drag=false)
   &VortexForce(
    UseOriginOffset=false
    Vortex Force Amount=&RandomRangeFloat(Minimum=800, Maximum=3000)
    Vortex Origin:NiagaraPosition=$Particles.SkeletalMeshLocation001.SampledPosition
   ) {enabled=false}
   &PointAttractionForce(
    Attraction Strength=&FloatFromCurve(
     FloatCurve=DataInterface() {propertyValues {Curve {defaultValue=1, keys [{tangentMode=RCTM_User, time=0, value=0}, {tangentMode=RCTM_User, time=1, value=1}]}}}
     Scale Curve=1000
    )
    Kill Within Radius=false
    Module.Use Falloff=true
    UseAttractorOffset=false
   ) {enabled=false}
   &SolveForcesAndVelocity(Clamp Velocity=false, Limit Acceleration=false)
   &ScaleColor(
    Scale Alpha=&FloatFromCurve(FloatCurve=DataInterface() {propertyValues {Curve {keys [<at 0 0>,<at 0.25 1>,<at 0.5 1>,<at 1 0>]}}})
    ScaleA=true
    ScaleRGB=true
   ) {enabled=false}
   &ScaleColor(
    Scale Alpha=&FloatFromCurve(FloatCurve=DataInterface() {propertyValues {Curve {keys [<at 0 0>,<at 0.1 1>,<at 0.9 1>,<at 1 0>]}}})
    Scale RGB=&FN_FrontEndVector3Modifier(Gameplay=<1 1 1>, Front End=&MakeVectorFromLinearColorRGB(Color=<rgba 2.5 0.6315 0.4322 1>))
    ScaleA=true
    ScaleRGB=true
   ) {enabled=false, module=ScaleColor001}
   &ScaleColor(
    Scale RGB=&FN_FrontEndVector3Modifier(
     Gameplay=&MakeVectorFromLinearColorRGB(Color=<rgba 0 0.5347 0.7 1>)
     Front End=&MakeVectorFromLinearColorRGB(Color=$User.FrontEnd_color)
    )
    ScaleA=true
    ScaleRGB=true
   ) {module=ScaleColor002}
   &ScaleMeshSize(Scale Factor=&VectorFromFloat(Value=&FloatFromCurve(FloatCurve=DataInterface() {propertyValues {Curve {keys [{arriveTangent=-1.3985, interpMode=RCIM_Cubic, leaveTangent=-1.3985, tangentMode=RCTM_User, time=0, value=1}, <at 0.9998 0.2>]}}}))) {module=ScaleMeshSize002}
   &CameraOffset()
  ]
  rendererClasses [&NiagaraMeshRendererProperties]
  renderers [{rendererClass=&NiagaraMeshRendererProperties, rendererIndex=0}]
 }
}
summary {
 systemName=NS_Emote_SpatialTravel_Hands
 emitters [
  {emitterName=NE_Glow, rendererClasses [&NiagaraSpriteRendererProperties, &NiagaraLightRendererProperties]}
  {emitterName=Light_Shafts, rendererClasses [&NiagaraSpriteRendererProperties, &NiagaraMeshRendererProperties]}
  {emitterName=Square_far001, rendererClasses [&NiagaraMeshRendererProperties]}
  {emitterName=BlowingParticles, rendererClasses [&NiagaraLightRendererProperties]}
 ]
 userVariables [
  User.AltTimeline {type=NiagaraFloat, defaultValue=0, description=""}
  User.bIsFrontEnd {type=NiagaraBool}
  User.bIsFrontEndPreview {type=NiagaraBool}
  User.CubeFalling {type=NiagaraFloat, defaultValue=0, description=""}
  User.CubeGlow {type=NiagaraFloat, defaultValue=0, description=""}
  User.CubeGlowStep {type=NiagaraFloat, defaultValue=0, description=""}
  User.CubeScale {type=NiagaraFloat, defaultValue=1, description=""}
  User.CubeStep {type=NiagaraFloat, defaultValue=0, description=""}
  User.FrontEnd_color {type=LinearColor, description="", defaultValue {struct=LinearColor, value=<rgba 1 0.6743 0.6445 1>}}
  User.HandGlow {type=NiagaraFloat, defaultValue=0, description=""}
  User.Height_Location {type=NiagaraFloat, defaultValue=17, description=""}
 ]
}
